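; Vehicle cosmetic actor for the Singularity boost effect, which sets up boost visuals when the cosmetic is applied to a car.
; On apply it swaps the boost Niagara component's system asset to its configured system and runs the boost count setup.
; A single-boost flag then picks between one-emitter positioning and per-car boost offsets obtained through the vehicle cosmetics interface.
; BeginPlay, overlap and Tick are empty stubs, and construction only calls the parent.
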(event EventBeginPlay)

(event Collision|EventActorBeginOverlap
  (params (OtherActor :type Actor%20Object%20Reference)))

(event EventTick
  (params (DeltaSeconds :type Float%20%28single-precision%29)))

(event Vehicle|Cosmetics|EventOnCosmeticApplied
  (Niagara|SetNiagaraSystemAsset (Variables|Default|GetNS_Boost) (Variables|Default|GetNiagaraSystem) true)
  (CallFunction|SetupBoostCount)
  (if (Variables|Default|GetSingleBoost?)
    (Class|BPVehicleCosmeticsFunctions|BPSingleBoostSetPosition (Variables|Default|GetNS_Boost))
    (Class|BPVehicleCosmeticsFunctions|SetBoostOffsetbyCar (Variables|Vehicle|Cosmetics|GetVehicleCosmeticsInterface))))

(fn ConstructionScript
  (|CallParentFunction))
