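# Master translucent unlit material for Niagara sprites in the Singularity vehicle boost cosmetic effect.
# A texture parameter is tinted by particle color and passed through the shared MF_FX_EmissiveAdjustments function for glow.
# A separate Opacity texture parameter is multiplied by particle alpha for transparency.
# Both parameters default to the same sphere gradient mask, giving a soft round sprite.
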Material /VehicleCosmetics/Boosts/Boost_Singularity/Materials/MM_Boost_Singularity
graph {
 properties {
  bAutomaticallySetUsageInEditor=false
  bUsedWithNiagaraSprites=true
  blendMode=BLEND_Translucent
  shadingModel=MSM_Unlit
  usedShadingModels=Unlit
 }
 nodes [
  MaterialExpressionVertexColor_0 = ParticleColor()
  MaterialFunctionCall(Emissive=Multiply(
    A=TextureSampleParameter2D() {parameterName=Texture, texture=/DealershipSharedLibrary/FX/Textures/Gradient_Masks/T_VCVFX_SphereGradient_A}.RGB
    B=@MaterialExpressionVertexColor_0.RGB
   )) {materialFunction=&MF_FX_EmissiveAdjustments}
  Multiply(
   A=@MaterialExpressionVertexColor_0.A
   B=TextureSampleParameter2D() {parameterName=Opacity, texture=&T_VCVFX_SphereGradient_A}.RGB
  )
 ]
}
