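# Additive unlit material for Niagara mesh particles that draws the black hole effect of the Singularity vehicle boost.
# Color adds a static black hole texture to panning rings, tints it by particle color and feeds an emissive adjustment function.
# Ring opacity combines oppositely panned masks with a particle-alpha-driven erosion term.
# Dynamic parameters BHAlpha and Distortion scale that ring mask and a distortion texture, respectively.
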Material /VehicleCosmetics/Boosts/Boost_Singularity/Materials/MM_Boost_Singularity_Blackhole
graph {
 properties {
  bAutomaticallySetUsageInEditor=false
  bUsedWithNiagaraMeshParticles=true
  blendMode=BLEND_Additive
  refractionMethod=RM_PixelNormalOffset
  shadingModel=MSM_Unlit
  usedShadingModels=Unlit
 }
 nodes [
  n1 = Panner(Coordinate=TextureCoordinate()) {speedX=0.2}
  MaterialExpressionMeshEmitterVertexColor_1 = ParticleColor()
  n2 = OneMinus(None=Subtract(
    A=@MaterialExpressionMeshEmitterVertexColor_1.A
    B=TextureSample(UVs=@n1) {texture=/VehicleCosmetics/Boosts/Boost_Singularity/Textures/T_VCVFX_Boost_Singularity_BlackholeRings_Opacity2}.RGB
   ))
  n3 = Constant()
  n4 = Constant() {r=1}
  n5 = Multiply(
   A=Multiply(
    A=Constant() {r=2}
    B=TextureSample(UVs=@n1) {texture=/VehicleCosmetics/Boosts/Boost_Singularity/Textures/T_VCVFX_Boost_Singularity_BlackholeRings2_Opacity}.RGB
   )
   B=Multiply(A=@n2, B=Clamp(None=@n2, Min=@n3, Max=@n4) {minDefault=1})
  )
  Add(A=TextureSample() {texture=/VehicleCosmetics/Boosts/Boost_Singularity/Textures/T_VCVFX_Boost_Singularity_Blackhole_Opacity}.RGB, B=@n5)
  n6 = TextureSample() {texture=/VehicleCosmetics/Boosts/Boost_Singularity/Textures/T_VCVFX_Boost_Singularity_BlackHole}
  n7 = TextureSample(UVs=Panner(Coordinate=TextureCoordinate()) {speedX=-0.2}) {texture=/VehicleCosmetics/Boosts/Boost_Singularity/Textures/T_VCVFX_Boost_Singularity_BlackholeRings2}
  MaterialFunctionCall(Emissive=Multiply(A=Add(A=@n6.RGB, B=@n7.RGB), B=@MaterialExpressionMeshEmitterVertexColor_1.RGB)) {materialFunction=&MF_FX_EmissiveAdjustments}
  MaterialExpressionMeshEmitterDynamicParameter_1 = DynamicParameter() {paramNames [BHAlpha,Distortion,Param3,Param4]}
  Multiply(
   A=TextureSample() {texture=/VehicleCosmetics/Boosts/Boost_Singularity/Textures/T_VCVFX_Boost_Singularity_Blackhole_Distortion2}.RGB
   B=@MaterialExpressionMeshEmitterDynamicParameter_1.Distortion
  )
  Multiply(A=@n5, B=@MaterialExpressionMeshEmitterDynamicParameter_1.BHAlpha)
 ]
}
