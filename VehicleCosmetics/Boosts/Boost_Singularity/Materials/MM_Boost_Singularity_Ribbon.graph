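# Translucent unlit Niagara ribbon material for the Singularity boost vehicle cosmetic, drawing a scrolling wave trail.
# Particle color RGB runs through MF_FX_EmissiveAdjustments for glow.
# A wave texture pans along U, with offset and speed driven per particle by Dynamic Parameters.
# The waves are multiplied by particle alpha, a trail transparency gradient and a squared U ramp, forming a mask that fades toward the U=0 end.
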Material /VehicleCosmetics/Boosts/Boost_Singularity/Materials/MM_Boost_Singularity_Ribbon
graph {
 properties {
  bAutomaticallySetUsageInEditor=false
  bUsedWithNiagaraRibbons=true
  blendMode=BLEND_Translucent
  shadingModel=MSM_Unlit
  usedShadingModels=Unlit
 }
 nodes [
  MaterialExpressionVertexColor_4 = ParticleColor()
  MaterialFunctionCall(Emissive=@MaterialExpressionVertexColor_4.RGB) {materialFunction=&MF_FX_EmissiveAdjustments}
  n1 = DynamicParameter() {paramNames [Offset X, Panner Velocity, Param3, Param4]}
  n2 = Constant() {r=0.275}
  n3 = Panner(Coordinate=Add(A=TextureCoordinate() {vTiling=0.45}, B=AppendVector(A=@n1.Offset X, B=@n2)), Time=Multiply(A=@n1.Panner Velocity, B=Time())) {speedX=-1}
  n4 = TextureSample() {texture=/DealershipSharedLibrary/FX/Textures/Gradient_Masks/T_VCVFX_SingularityTrailTransparency_A}
  n5 = ComponentMask(None=TextureCoordinate()) {r=true}
  n6 = Constant() {r=2}
  Multiply(
   A=Multiply(
    A=TextureSample(UVs=@n3) {texture=/VehicleCosmetics/Boosts/Boost_Singularity/Textures/T_VCVFX_Boost_Singularity_Waves}.RGB
    B=Multiply(A=@MaterialExpressionVertexColor_4.A, B=@n4.RGB)
   )
   B=Multiply(A=Power(Base=@n5, Exp=@n6)) {constB=2}
  )
 ]
}
